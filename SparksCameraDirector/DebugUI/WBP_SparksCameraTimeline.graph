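# Debug UMG timeline for the Sparks camera director, showing song progress and recorded camera shots across a performance map.
# Tick fills the progress bar from the playspace camera data provider's current event time divided by performance length.
# OnPaint converts milliseconds to bar pixels, then draws tagged section markers, start/end markers, labels and shot lines colored and offset by source.
# AddShot appends shots; SetPerfMap calls Reset, which clears them, before storing the map.
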WidgetBlueprint /SparksCameraDirector/DebugUI/WBP_SparksCameraTimeline
graphs [
 AddShot {nodes [
   n1 = AddShot!() -> (then, Shot:"Sparks Camera Director Entry Runtime Data Structure (by ref)") {is=title K2Node_FunctionEntry}
   > Add(
    TargetArray:Array of Sparks Camera Director Entry Runtime Data Structures=Get Shots() -> (Shots:Array of Sparks Camera Director Entry Runtime Data Structures)
    NewItem:"Sparks Camera Director Entry Runtime Data Structure (by ref)"=@n1
   ) {is=title K2Node_CallArrayFunction}
  ]}
 DrawLabelText {nodes [
   n1 = Inputs!() -> (execute, Context:"Paint Context Structure (by ref)", Text:"Text (by ref)", YOffset:double=~, Type:ESparksCameraShotSource Enum=Dynamic)
   n2 = Get!() -> (ProgPos_X:double, ProgPos_Y:double)
   > Draw Text!(
    execute
    Context:"Paint Context Structure (by ref)"=@n1
    Text:"Text (by ref)"=@n1.Text
    Position_X:double=float - float(A=@n2, B=400)
    Position_Y:double=float + float(A=@n2.ProgPos_Y, B=@n1.YOffset)
    Font:Font Object Reference=/LauncherSocial/BrutalType
    FontSize:float=10
    FontTypeFace:Name=Regular
    Tint:Linear Color Structure=Get Text Color!(Index:ESparksCameraShotSource Enum=@n1.Type) -> (ReturnValue:Linear Color Structure) {is=title K2Node_MacroInstance}
   ) -> (then)
   > Outputs!(Exec:Exec)
  ]}
 DrawPosMarker {nodes [
   n1 = Inputs!() -> (execute, Context, MS:double=~)
   n2 = Get()
   n3 = float + float(A=float * float(A=Get ProgXStep() -> (ProgXStep:double), B=@n1.MS) {is=title K2Node_PromotableOperator}, B=@n2)
   > DrawLine!(
    execute
    Context:"Paint Context Structure (by ref)"=@n1
    PositionA_X:double=@n3
    PositionA_Y:double=float - float(A=@n2.ProgPos_Y, B=20)
    PositionB_X:double=@n3
    PositionB_Y:double=float + float!(A=@n2.ProgPos_Y, B=20, C:double=Get!() -> (ProgSize_X:double, ProgSize_Y:double).ProgSize_Y) -> (ReturnValue)
    Tint:Linear Color Structure=(R=1,G=0.1406,B=0.1406,A=1)
    bAntiAlias:bool=true
    Thickness:float=3
   ) -> (then)
   > Outputs!(then:Exec)
  ]}
 DrawSectionMarker {nodes [
   n1 = Inputs!() -> (execute, Context, MS=~, Text:"Text (by ref)")
   n2 = Get!() -> (ProgPos_X:double, ProgPos_Y:double)
   n3 = float + float(A=@n2.ProgPos_Y, B=20, C=Get!() -> (ProgSize_X:double, ProgSize_Y:double).ProgSize_Y)
   n4 = float + float!(A=float * float(A=Get ProgXStep(), B=@n1.MS), B=@n2) -> (ReturnValue)
   > DrawLine(
    Context=@n1
    PositionA_X=@n4
    PositionA_Y=float - float(A=@n2.ProgPos_Y, B=20)
    PositionB_X=@n4
    PositionB_Y=@n3
    Tint=(R=1,G=0.1406,B=0.1406,A=1)
    bAntiAlias=true
    Thickness=1
   )
   > Draw Text(
    Context=@n1
    Text=@n1.Text
    Position_X=float + float(A=@n4, B=5)
    Position_Y=float - float(A=@n3, B=10)
    Font=&BrutalType
    FontSize=10
    FontTypeFace=Regular
    Tint=(R=0.6302,G=1,B=0.9905,A=1)
   )
   > Outputs!(out:Exec)
  ]}
 DrawShotBox {nodes [
   n1 = Inputs!() -> (execute, Context, SparksCameraDirectorEntryRuntimeData:"Sparks Camera Director Entry Runtime Data Structure (by ref)")
   n2 = Break Sparks Camera Director Entry Runtime Data!(SparksCameraDirectorEntryRuntimeData:"Sparks Camera Director Entry Runtime Data Structure (by ref)"=@n1.SparksCameraDirectorEntryRuntimeData) -> (
    Entry:Sparks Camera Director Sequence Entry Structure
    GeneratorName:Name
    SectionName:Name
    Config:Sparks Camera Director Config Object Reference
    NowAtStart:float=~
    Source:ESparksCameraShotSource Enum
    MSOffset:float=~
    LengthSeconds:float=~
    PlayTime:float=~
    bIsDynamic:bool=~
    bPreferLocalPlayer:bool=~
    ForceTarget:Actor Object Reference
    TargetActor:Actor Object Reference
    TargetLookAtActor:Actor Object Reference
    TargetAttachMesh:Skeletal Mesh Component Object Reference
    TargetLookAtMesh:Skeletal Mesh Component Object Reference
    Result:ESparksCamera_PlayShotResult Enum
    LogSteps:Array of Strings
   ) {is=title K2Node_BreakStruct}
   > n3 = Branch(Condition="float >= float"(A=@n2.NowAtStart))
   n4 = Get ProgXStep()
   n5 = Get()
   n6 = float * float(A=@n5, B=-1)
   n7 = float * float(A=@n5.ProgSize_Y, B=2)
   n8 = Get!() -> (ProgPos_X:double, ProgPos_Y:double)
   > n9 = DrawBox!(
    execute
    Context:"Paint Context Structure (by ref)"=@n1
    Position_X:double=float + float(A=float * float(A=@n4, B=@n2.NowAtStart), B=@n8)
    Position_Y:double=float + float(
     A=Select!(Dynamic:double=@n6, Scripted:double=0, Intermission:double=@n5.ProgSize_Y, Unknown:double=@n7, Index:ESparksCameraShotSource Enum=@n2.Source) -> (ReturnValue:double)
     B=@n8.ProgPos_Y
    )
    Size_X:double=float * float!(A=@n4, B=@n2.PlayTime, C:double=1000) -> (ReturnValue)
    Size_Y:double=@n5.ProgSize_Y
    Brush:Slate Brush Asset Object Reference=/Game/UI/CodeReferenced/Tooltip-Micro_Brush
    Tint:Linear Color Structure=Select(
     Dynamic:Linear Color Structure=(R=1,G=0.9249,B=0.1532,A=1)
     Scripted:Linear Color Structure=(R=0.2587,G=0.3642,B=1,A=1)
     Intermission:Linear Color Structure=(R=0.295,G=1,B=0.332,A=1)
     Unknown:Linear Color Structure
     Index=@n2.Source
    ) -> (ReturnValue:Linear Color Structure)
   ) -> (then)
   Outputs!(Exec:Exec=[@n9.then, @n3.else])
  ]}
 DrawShotLine {nodes [
   n1 = Inputs()
   n2 = Break Sparks Camera Director Entry Runtime Data(SparksCameraDirectorEntryRuntimeData=@n1.SparksCameraDirectorEntryRuntimeData) -> (NowAtStart=~, MSOffset=~, LengthSeconds=~, PlayTime=~, bIsDynamic=~, bPreferLocalPlayer=~)
   > n3 = Branch(Condition="float >= float"(A=@n2.NowAtStart))
   n4 = Get()
   n5 = float - float(
    A=@n4.ProgPos_Y
    B=Select(Dynamic:double=15, Scripted:double=10, Intermission:double=5, Unknown:double=1, Index=@n2.Source) -> (ReturnValue:double)
   )
   n6 = Get ProgXStep()
   n7 = float + float(A=float * float!(A=@n6, B=@n2.NowAtStart) -> (ReturnValue), B=@n4)
   > n8 = DrawLine(
    Context=@n1
    PositionA_X=@n7
    PositionA_Y=@n5
    PositionB_X=float + float(A=@n7, B=float * float!(A=@n6, B=@n2.PlayTime, C:double=1000) -> (ReturnValue))
    PositionB_Y=@n5
    Tint=Get Text Color(Index=@n2.Source)
    bAntiAlias=true
    Thickness=5
   )
   Outputs(Exec=[@n8.then, @n3.else])
  ]}
 EventGraph {nodes [
   Event Tick()
   n1 = Get PerfMap() -> (PerfMap:Sparks Performance Map Object Reference)
   > Is Valid(InputObject=@n1)
   n2 = GetOwningPlayerPawn!(self:User Widget Object Reference) -> (ReturnValue:Pawn Object Reference)
   > n3 = GetPlayspaceForActor!(execute, WorldContextObject:Object Reference=@n2, InActor:Actor Object Reference=@n2) -> (then, ReturnValue:Playspace Object Reference)
   > Is Valid(InputObject=@n3)
   > n4 = GetCurrentMSTimeInCameraEvent!(
    execute
    self:Sparks Camera Plugin Data Provider Object Reference=GetComponentByClass!(
     self:Actor Object Reference=@n3
     ComponentClass:Actor Component Class Reference=/Script/SparksCameraDirectorRuntime.SparksCameraPluginDataProvider
    ) -> (ReturnValue:Sparks Camera Plugin Data Provider Object Reference)
   ) -> (then, ReturnValue:float)
   > SetPercent!(
    execute
    self:Progress Bar Object Reference=Get prog_Song() -> (prog_Song:Progress Bar Object Reference)
    InPercent:float=float / float(A=@n4, B=Get PerformanceLength!(self:Sparks Performance Map Object Reference=@n1) -> (PerformanceLength:float))
   ) -> (then)
   Event PreConstruct()
   Event Construct()
  ]}
 GetTextColor {nodes [Outputs!(ReturnValue:Linear Color Structure=Select(
     Dynamic:Linear Color Structure=(R=1,G=0.9047,B=0,A=1)
     Scripted:Linear Color Structure=(R=0.1878,G=0.305,B=1,A=1)
     Intermission:Linear Color Structure=(R=0.2051,G=1,B=0.2462,A=1)
     Unknown:Linear Color Structure=(R=1,G=0,B=0,A=1)
     Index=Inputs!() -> (Index:ESparksCameraShotSource Enum=Dynamic)
    ) -> (ReturnValue:Linear Color Structure))]}
 OnPaint {nodes [
   n1 = Get prog_Song()
   n2 = GetCachedGeometry!(self:Widget Object Reference=@n1) -> (ReturnValue:"Geometry Structure (by ref)")
   n3 = Set ProgSize!(
    execute=@n8.then_0
    ProgSize:Vector 2D Structure=GetLocalSize!(Geometry:"Geometry Structure (by ref)"=@n2) -> (ReturnValue:Vector 2D Structure)
   ) -> (then, Output_Get_X:double, Output_Get_Y:double)
   PrintText!(
    execute
    InText=Format Text!(Format="LOCTEXT(\"Prog Size: {x}, {y}\")", x:double=@n3, y:double=@n3.Output_Get_Y) -> (Result)
    bPrintToScreen
    bPrintToLog
    TextColor=(R=1,G=0,B=0.9217,A=1)
    Duration
    Key=CameraDBG_size
   ) -> (then)
   > PrintText(
    InText=Format Text!(Format="LOCTEXT(\"Prog XStep: {x}\")", x=Get ProgXStep()) -> (Result)
    TextColor=(R=1,G=0,B=0.9217,A=1)
    Key=CameraDBG_XStep
   )
   n4 = Get()
   > PrintText(
    InText=Format Text!(Format="LOCTEXT(\"Prog Pos: {x}, {y}\")", x=@n4, y:double=@n4.ProgPos_Y) -> (Result)
    TextColor=(R=1,G=0,B=0.9217,A=1)
    Key=CameraDBG_pos
   )
   n5 = OnPaint!() -> (then, Context:"Paint Context Structure (by ref)") {is=title K2Node_FunctionEntry}
   n6 = Get PerfMap()
   > n7 = Is Valid(InputObject=@n6)
   > n8 = Sequence!(execute) -> (then_0, then_1, then_2:Exec, then_3:Exec, then_4:Exec)
   n9 = For Each Loop(
    Exec=@n8.then_1
    Array:Array of Sparks Performance Map Section Structures=Get Timeline!(self:Sparks Performance Map Object Reference=Get PerfMap()) -> (Timeline:Array of Sparks Performance Map Section Structures)
   ) -> (Array Element:Sparks Performance Map Section Structure)
   n10 = Break Sparks Performance Map Section!(SparksPerformanceMapSection:"Sparks Performance Map Section Structure (by ref)"=@n9) -> (
    StartMS:float=~
    LengthMS:float=~
    PrimaryTag:Gameplay Tag Structure
    Tags:Gameplay Tag Container Structure
    ShotsFilter:Sparks Camera Filter Structure
   ) {is=title K2Node_BreakStruct}
   n11 = GetDebugStringFromGameplayTag!(GameplayTag:Gameplay Tag Structure=@n10.PrimaryTag) -> (ReturnValue:String)
   > Draw Section Marker!(
    execute:Exec
    Context:"Paint Context Structure (by ref)"=@n5
    MS:double=@n10
    Text:"Text (by ref)"="To Text (String)"(InString=Split(SourceString=@n11, InStr=., SearchDir=FromEnd).RightS)
   ) -> (out:Exec) {is=title K2Node_MacroInstance}
   n12 = For Each Loop(Exec=@n8.then_2, Array:Array of Sparks Camera Director Entry Runtime Data Structures=Get Shots()) -> (Array Element:Sparks Camera Director Entry Runtime Data Structure)
   > Draw Shot Line!(
    execute:Exec
    Context:"Paint Context Structure (by ref)"=@n5
    SparksCameraDirectorEntryRuntimeData:"Sparks Camera Director Entry Runtime Data Structure (by ref)"=@n12
   ) -> (Exec:Exec) {is=title K2Node_MacroInstance}
   Draw Label Text!(
    execute:Exec=@n8.then_3
    Context:"Paint Context Structure (by ref)"=@n5
    Text:"Text (by ref)"=MakeLiteralText(Value="LOCTEXT(\"Dynamic:\")")
    YOffset:double=-15
    Type:ESparksCameraShotSource Enum=Dynamic
   ) -> (Exec:Exec) {is=title K2Node_MacroInstance}
   > Draw Label Text(Context=@n5, Text=MakeLiteralText(Value="LOCTEXT(\"Scripted:\")"), Type=Scripted)
   > Draw Label Text(Context=@n5, Text=MakeLiteralText(Value="LOCTEXT(\"Intermission:\")"), YOffset=15, Type=Intermission)
   n13 = Get PerformanceLength(self=@n6)
   Draw Pos Marker!(execute:Exec=@n8.then_4, Context:"Paint Context Structure (by ref)"=@n5, MS:double=@n13) -> (then:Exec) {is=title K2Node_MacroInstance}
   > n14 = Draw Pos Marker(Context=@n5)
   Return Node(execute=[@n14.then, @n7.Is Not Valid])
   n15 = SlotAsCanvasSlot!(Widget:Widget Object Reference=@n1) -> (ReturnValue:Canvas Panel Slot Object Reference)
   GetSize!(self:Canvas Panel Slot Object Reference=@n15) -> (ReturnValue:Vector 2D Structure)
   n16 = Get()
   n17 = float + float(A=float * float!(A=Get ProgXStep(), B) -> (ReturnValue), B=@n16)
   DrawLine(
    PositionA_X=@n17
    PositionA_Y=float - float(A=@n16.ProgPos_Y, B=30)
    PositionB_X=@n17
    PositionB_Y=float + float!(A=@n16.ProgPos_Y, B=30, C:double=Get!() -> (ProgSize_X:double, ProgSize_Y:double).ProgSize_Y) -> (ReturnValue)
    Tint=(R=1,G=0.1406,B=0.1406,A=1)
    bAntiAlias=true
    Thickness=3
   )
   GetAbsoluteSize!(Geometry:"Geometry Structure (by ref)"=@n2) -> (ReturnValue:Vector 2D Structure)
   GetAbsolutePosition!(Geometry:"Geometry Structure (by ref)"=@n2) -> (ReturnValue:Vector 2D Structure)
   n18 = Break Sparks Camera Director Entry Runtime Data() -> (NowAtStart=~, MSOffset=~, LengthSeconds=~, PlayTime=~, bIsDynamic=~, bPreferLocalPlayer=~)
   n19 = Get ProgXStep()
   n20 = Get!() -> (ProgPos_X:double, ProgPos_Y:double)
   n21 = Get!() -> (ProgSize_X:double, ProgSize_Y:double)
   n22 = float * float(A=@n21.ProgSize_Y, B=2)
   n23 = float * float(A=@n21, B=-1)
   DrawBox(
    Position_X=float + float!(A=float * float(A=@n19, B=@n18.NowAtStart), B=@n20) -> (ReturnValue)
    Position_Y=float + float(
     A=Select(Dynamic:double=@n23, Scripted:double=0, Intermission:double=@n21.ProgSize_Y, Unknown:double=@n22, Index=@n18.Source) -> (ReturnValue:double)
     B=@n20.ProgPos_Y
    )
    Size_X=float * float!(A=@n19, B=@n18.PlayTime, C:double=0.001) -> (ReturnValue)
    Size_Y=@n21.ProgSize_Y
    Brush=/Game/Blueprints/Blutilities/ForCinematics/PlaceByUVTool/drawTestBrush
    Tint=Select(
     Dynamic:Linear Color Structure=(R=1,G=0.9249,B=0.1532,A=1)
     Scripted:Linear Color Structure=(R=0.2587,G=0.3642,B=1,A=1)
     Intermission:Linear Color Structure=(R=0.295,G=1,B=0.332,A=1)
     Unknown:Linear Color Structure
     Index=@n18.Source
    ) -> (ReturnValue:Linear Color Structure)
   )
   n24 = Break Sparks Camera Director Entry Runtime Data() -> (NowAtStart=~, MSOffset=~, LengthSeconds=~, PlayTime=~, bIsDynamic=~, bPreferLocalPlayer=~)
   n25 = Get ProgXStep()
   n26 = Get!() -> (ProgPos_X:double, ProgPos_Y:double)
   n27 = float + float(A=float * float!(A=@n25, B=@n24.NowAtStart) -> (ReturnValue), B=@n26)
   DrawLine(
    PositionA_X=@n27
    PositionA_Y=@n26.ProgPos_Y
    PositionB_X=float + float(A=@n27, B=float * float!(A=@n25, B=@n24.PlayTime, C:double=1000) -> (ReturnValue))
    PositionB_Y=@n26.ProgPos_Y
    Tint=(R=0,G=0.7378,B=0.1684,A=1)
    bAntiAlias=true
    Thickness=5
   )
   Draw Shot Box!(
    execute:Exec
    Context:"Paint Context Structure (by ref)"=@n5
    SparksCameraDirectorEntryRuntimeData:"Sparks Camera Director Entry Runtime Data Structure (by ref)"=@n12
   ) -> (Exec:Exec) {is=title K2Node_MacroInstance}
   n28 = Get!() -> (ProgSize_X:double, ProgSize_Y:double)
   float + float(
    A=Select(
     Dynamic:double=float * float!(A=@n28, B=-1) -> (ReturnValue)
     Scripted:double=0
     Intermission:double=@n28.ProgSize_Y
     Unknown:double=float * float(A=@n28.ProgSize_Y, B=2)
     Index=@n24.Source
    ) -> (ReturnValue:double)
    B=@n26.ProgPos_Y
   )
   Select(
    Dynamic:Linear Color Structure=(R=1,G=0.9249,B=0.1532,A=1)
    Scripted:Linear Color Structure=(R=0.2587,G=0.3642,B=1,A=1)
    Intermission:Linear Color Structure=(R=0.295,G=1,B=0.332,A=1)
    Unknown:Linear Color Structure
    Index=@n24.Source
   ) -> (ReturnValue:Linear Color Structure)
   Set ProgPos(
    execute=@n30.then
    ProgPos:Vector 2D Structure=GetPosition!(self:Canvas Panel Slot Object Reference=@n15) -> (ReturnValue:Vector 2D Structure)
   ) -> (Output_Get:Vector 2D Structure)
   n29 = float / float(A=@n3, B=@n13)
   n30 = Set ProgXStep(execute=@n3.then, ProgXStep:double=@n29) -> (Output_Get:double)
  ]}
 Reset {nodes [
   Reset() {is=title K2Node_FunctionEntry}
   > SetProgress!(execute, self, Percent:double=0) -> (then)
   > Set SongLength(SongLength:double=0) -> (Output_Get:double)
   > n1 = Sequence!(execute) -> (then_0, then_1)
   > Clear(TargetArray:Array of Sparks Camera Director Entry Runtime Data Structures=Get Shots())
   n2 = Get PerfMap()
   Is Valid(exec=@n1.then_1, InputObject=@n2)
   > Set SongLength(SongLength=Get PerformanceLength(self=@n2))
  ]}
 SetPerfMap {nodes [
   n1 = SetPerfMap!() -> (then, Map:Sparks Performance Map Object Reference) {is=title K2Node_FunctionEntry}
   > Reset() {is=title K2Node_CallFunction}
   > Set PerfMap(PerfMap:Sparks Performance Map Object Reference=@n1) -> (Output_Get:Sparks Performance Map Object Reference)
  ]}
 SetProgress {nodes [
   n1 = SetProgress!() -> (then, Percent:double=~) {is=title K2Node_FunctionEntry}
   > SetPercent(self=Get prog_Song(), InPercent=@n1)
  ]}
]
variables [
 PerfMap {type=SparksPerformanceMap, default=null}
 Shots {type="array<SparksCameraDirectorEntryRuntimeData>", default []}
 SongLength {type=number}
]
widgets {
 info {inheritedWidgetCount=0, parentClass=&UserWidget, rootWidgetClass=&SizeBox, widgetCount=4}
 widgets [
  SizeBox_0 {
   widgetClassPath=&SizeBox
   widget=SizeBox_0
   properties {
    bOverride_HeightOverride=true
    bOverride_MaxDesiredHeight=true
    bOverride_MinDesiredHeight=true
    heightOverride=300
    maxDesiredHeight=300
    minDesiredHeight=300
   }
  }
  CanvasPanel_0 {widgetClassPath=&CanvasPanel, parent=SizeBox_0, widget=CanvasPanel_0, slot {class=SizeBoxSlot}}
  Image {
   widgetClassPath=&Image
   bIsVariable=true
   parent=CanvasPanel_0
   widget=Image
   properties {colorAndOpacity=<rgba 0 0 0 0.33>}
   slot {class=CanvasPanelSlot, properties {layoutData {alignment=<0 0>, anchors {maximum=<1 1>, minimum=<0 0>}, offsets=<ltrb 0.654 0 0 0>}}}
  }
  prog_Song {
   widgetClassPath=/Script/UMG.ProgressBar
   bIsVariable=true
   parent=CanvasPanel_0
   widget=prog_Song
   properties {fillColorAndOpacity=<rgba 0 0.5 1 1>}
   slot {class=CanvasPanelSlot, properties {layoutData {alignment=<0 0>, anchors {maximum=<1 0>, minimum=<0 0>}, offsets=<ltrb 48 48 48 15.991>}}}
  }
 ]
}
