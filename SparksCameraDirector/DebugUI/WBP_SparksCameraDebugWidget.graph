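# Debug overlay widget for the Sparks Camera Director, showing the current generator, section and playing clip plus a camera shot timeline.
# When given a Gameplay Event Router, it listens for log, progress, performance map and shot played events.
# Log events update the text line matching their log type.
# The other events feed progress, the performance map and played shots into the embedded WBP_SparksCameraTimeline.
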WidgetBlueprint /SparksCameraDirector/DebugUI/WBP_SparksCameraDebugWidget
graphs [
 EventGraph {nodes [
   n1 = Event SetEventRouter!() -> (OutputDelegate, then, EventRouter:Gameplay Event Router Component Object Reference) {is=title K2Node_Event}
   > n2 = Set Event Router(Event Router:Gameplay Event Router Component Object Reference=@n1.EventRouter) -> (Output_Get:Gameplay Event Router Component Object Reference)
   > Is Valid(InputObject=@n2)
   > n3 = StartListeningToEvent!(
    execute:Exec
    Target:Gameplay Event Router Component Object Reference=@n2
    EventType:Script Struct Object Reference=/Script/SparksCameraDirectorRuntime.SparksCommonCameraDebugLogEvent
    Context:Object Reference
    EventBubblingRule:EEventBubblingRule Enum
   ) -> (
    then:Exec
    AsyncTaskProxy:Async Action Start Listening To Event Object Reference
    OnEventReceived:Exec
    Payload:Sparks Common Camera Debug Log Event Structure
   )
   > n4 = StartListeningToEvent(Target=@n2, EventType=/Script/SparksCameraDirectorRuntime.SparksCommonCameraDebugLogProgressEvent) -> (Payload:Sparks Common Camera Debug Log Progress Event Structure)
   > n5 = StartListeningToEvent(Target=@n2, EventType=/Script/SparksCameraDirectorRuntime.SparksCommonCameraPerformanceMapEvent) -> (Payload:Sparks Common Camera Performance Map Event Structure)
   > n6 = StartListeningToEvent(Target=@n2, EventType=/Script/SparksCameraDirectorRuntime.SparksCommonCameraShotPlayedEvent) -> (Payload:Sparks Common Camera Shot Played Event Structure)
   AddShot!(
    execute=@n6.OnEventReceived
    self:WBP Sparks Camera Timeline Object Reference=Get WBP_SparksCameraTimeline() -> (WBP_SparksCameraTimeline:WBP Sparks Camera Timeline Object Reference)
    Shot:"Sparks Camera Director Entry Runtime Data Structure (by ref)"=Break Sparks Common Camera Shot Played Event!(SparksCommonCameraShotPlayedEvent:"Sparks Common Camera Shot Played Event Structure (by ref)"=@n6.Payload) -> (Shot:Sparks Camera Director Entry Runtime Data Structure) {is=title K2Node_BreakStruct}
   ) -> (then)
   SetPerfMap!(
    execute=@n5.OnEventReceived
    self:WBP Sparks Camera Timeline Object Reference=Get WBP_SparksCameraTimeline()
    Map:Sparks Performance Map Object Reference=Break Sparks Common Camera Performance Map Event!(SparksCommonCameraPerformanceMapEvent:"Sparks Common Camera Performance Map Event Structure (by ref)"=@n5.Payload) -> (Map:Sparks Performance Map Object Reference) {is=title K2Node_BreakStruct}
   ) -> (then)
   SetProgress!(
    execute=@n4.OnEventReceived
    self:WBP Sparks Camera Timeline Object Reference=Get WBP_SparksCameraTimeline()
    Percent:double=Break Sparks Common Camera Debug Log Progress Event!(SparksCommonCameraDebugLogProgressEvent:"Sparks Common Camera Debug Log Progress Event Structure (by ref)"=@n4.Payload) -> (Value:float=~) {is=title K2Node_BreakStruct}
   ) -> (then)
   n7 = Break Sparks Common Camera Debug Log Event!(SparksCommonCameraDebugLogEvent:"Sparks Common Camera Debug Log Event Structure (by ref)"=@n3.Payload) -> (Data:String, Type:ESparksCameraLogType Enum) {is=title K2Node_BreakStruct}
   LogToScreen!(execute=@n3.OnEventReceived, self:Sparks Camera Debug Widget Object Reference, Data:String=@n7, Type:ESparksCameraLogType Enum=@n7.Type) -> (then)
   n8 = Event LogToScreen!() -> (OutputDelegate, then, Data:String, Type:ESparksCameraLogType Enum=ShotGenerator) {is=title K2Node_Event}
   > n9 = GetText!(execute, self, InString:String=@n8.Data, Index:ESparksCameraLogType Enum=@n8.Type) -> (then, NewParam:Text)
   > Set Text(
    Text=@n9
    self=Select!(
     ShotGenerator:Common Text Object Reference=Get txtGenerator() -> (txtGenerator:Common Text Object Reference)
     ShotName:Common Text Object Reference=Get txtClip() -> (txtClip:Common Text Object Reference)
     SectionName:Common Text Object Reference=Get txtSection() -> (txtSection:Common Text Object Reference)
     Unknown:Common Text Object Reference
     Index:ESparksCameraLogType Enum=@n8.Type
    ) -> (ReturnValue:Common Text Object Reference)
   )
   Event ClearAll() {is=title K2Node_Event}
   > ClearCurrent(self:Sparks Camera Debug Widget Object Reference)
   Event ClearCurrent() {is=title K2Node_Event}
   > Set Text(Text="INVTEXT(\"Generator:\")", self=Get txtGenerator())
   > Set Text(Text="INVTEXT(\"Section:\")", self=Get txtSection())
   > Set Text(Text="INVTEXT(\"Clip:\")", self=Get txtClip())
   Event Construct()
  ]}
 GetText {nodes [
   n1 = GetText!() -> (then, InString:String, Index:ESparksCameraLogType Enum=ShotSource) {is=title K2Node_FunctionEntry}
   n2 = "To Text (String)"(InString=@n1)
   > Return Node!(
    execute
    NewParam:Text=Select(
     ShotGenerator:Text=Format Text!(Format="INVTEXT(\"Generator: {gen}\")", gen:Text=@n2) -> (Result)
     ShotName:Text=Format Text!(Format="INVTEXT(\"Playing Clip: {clip}\")", clip:Text=@n2) -> (Result)
     SectionName:Text=Format Text!(Format="INVTEXT(\"Section: {section}\")", section:Text=@n2) -> (Result)
     Unknown:Text
     Index=@n1.Index
    ) -> (ReturnValue:Text)
   )
  ]}
]
variables [Event Router {type=GameplayEventRouterComponent, default=null}]
widgets {
 info {
  inheritedWidgetCount=0
  parentClass=/Script/SparksCameraDirectorRuntime.SparksCameraDebugWidget
  rootWidgetClass=&CanvasPanel
  widgetCount=7
 }
 widgets [
  CanvasPanel_43 {widgetClassPath=&CanvasPanel, widget=CanvasPanel_43}
  CanvasPanel_45 {
   widgetClassPath=&CanvasPanel
   parent=CanvasPanel_43
   widget=CanvasPanel_45
   slot {class=CanvasPanelSlot, properties {layoutData {alignment=<0 0>, anchors {maximum=<1 1>, minimum=<1 1>}, offsets=<ltrb -750 -350 700 100>}}}
  }
  txtSection {
   widgetClassPath=&CommonTextBlock
   bIsVariable=true
   parent=CanvasPanel_45
   widget=txtSection
   properties {
    lineHeightPercentage=1.28
    style=/FortUILibrary/Styles/Typography/_TextStyle_Body2._TextStyle_Body2_C
    text="Section: "
    font {
     bForceMonospaced=false
     bMaterialIsStencil=false
     fontMaterial=null
     fontObject=&BurbankSmall
     letterSpacing=0
     monospacedWidth=1
     size=15
     skewAmount=0
     typefaceFontName=Bold
     outlineSettings {
      bApplyOutlineToDropShadows=false
      bMiteredCorners=false
      bSeparateFillAlpha=false
      outlineMaterial=null
      outlineSize=0
      outlineColor=<rgba 0 0 0 1>
     }
    }
    shadowOffset=<0 0>
    strikeBrush {
     bIsDynamicallyLoaded=false
     drawAs=NoDrawType
     imageType=NoImage
     mirroring=NoMirror
     resourceName=None
     resourceObject=null
     tiling=NoTile
     imageSize=<32 32>
     margin=<ltrb 0 0 0 0>
     outlineSettings {
      bUseBrushTransparency=false
      roundingType=HalfHeightRadius
      width=0
      color {colorUseRule=UseColor_Specified, specifiedColor=<rgba 0 0 0 0>}
      cornerRadii=<0 0 0 0>
     }
     tintColor {colorUseRule=UseColor_Specified, specifiedColor=<rgba 1 1 1 1>}
     uVRegion {bIsValid=false, max=<0 0>, min=<0 0>}
    }
   }
   slot {class=CanvasPanelSlot, properties {layoutData {alignment=<0 0>, anchors {maximum=<1 0>, minimum=<0 0>}, offsets=<ltrb 4 4 4 28>}}}
  }
  txtClip {
   widgetClassPath=&CommonTextBlock
   bIsVariable=true
   parent=CanvasPanel_45
   widget=txtClip
   properties {
    lineHeightPercentage=1.28
    style=&_TextStyle_Body2_C
    text="Clip: "
    font {
     bForceMonospaced=false
     bMaterialIsStencil=false
     fontMaterial=null
     fontObject=&BurbankSmall
     letterSpacing=0
     monospacedWidth=1
     size=15
     skewAmount=0
     typefaceFontName=Bold
     outlineSettings {
      bApplyOutlineToDropShadows=false
      bMiteredCorners=false
      bSeparateFillAlpha=false
      outlineMaterial=null
      outlineSize=0
      outlineColor=<rgba 0 0 0 1>
     }
    }
    shadowOffset=<0 0>
    strikeBrush {
     bIsDynamicallyLoaded=false
     drawAs=NoDrawType
     imageType=NoImage
     mirroring=NoMirror
     resourceName=None
     resourceObject=null
     tiling=NoTile
     imageSize=<32 32>
     margin=<ltrb 0 0 0 0>
     outlineSettings {
      bUseBrushTransparency=false
      roundingType=HalfHeightRadius
      width=0
      color {colorUseRule=UseColor_Specified, specifiedColor=<rgba 0 0 0 0>}
      cornerRadii=<0 0 0 0>
     }
     tintColor {colorUseRule=UseColor_Specified, specifiedColor=<rgba 1 1 1 1>}
     uVRegion {bIsValid=false, max=<0 0>, min=<0 0>}
    }
   }
   slot {class=CanvasPanelSlot, properties {layoutData {alignment=<0 0>, anchors {maximum=<1 0>, minimum=<0 0>}, offsets=<ltrb 4 32 4 28>}}}
  }
  txtGenerator {
   widgetClassPath=&CommonTextBlock
   bIsVariable=true
   parent=CanvasPanel_45
   widget=txtGenerator
   properties {
    lineHeightPercentage=1.28
    style=&_TextStyle_Body2_C
    text="Generator: "
    font {
     bForceMonospaced=false
     bMaterialIsStencil=false
     fontMaterial=null
     fontObject=&BurbankSmall
     letterSpacing=0
     monospacedWidth=1
     size=15
     skewAmount=0
     typefaceFontName=Bold
     outlineSettings {
      bApplyOutlineToDropShadows=false
      bMiteredCorners=false
      bSeparateFillAlpha=false
      outlineMaterial=null
      outlineSize=0
      outlineColor=<rgba 0 0 0 1>
     }
    }
    shadowOffset=<0 0>
    strikeBrush {
     bIsDynamicallyLoaded=false
     drawAs=NoDrawType
     imageType=NoImage
     mirroring=NoMirror
     resourceName=None
     resourceObject=null
     tiling=NoTile
     imageSize=<32 32>
     margin=<ltrb 0 0 0 0>
     outlineSettings {
      bUseBrushTransparency=false
      roundingType=HalfHeightRadius
      width=0
      color {colorUseRule=UseColor_Specified, specifiedColor=<rgba 0 0 0 0>}
      cornerRadii=<0 0 0 0>
     }
     tintColor {colorUseRule=UseColor_Specified, specifiedColor=<rgba 1 1 1 1>}
     uVRegion {bIsValid=false, max=<0 0>, min=<0 0>}
    }
   }
   slot {class=CanvasPanelSlot, properties {layoutData {alignment=<0 0>, anchors {maximum=<1 0>, minimum=<0 0>}, offsets=<ltrb 4 60 4 28>}}}
  }
  Image_0 {
   widgetClassPath=&Image
   bIsVariable=true
   parent=CanvasPanel_45
   widget=Image_0
   properties {colorAndOpacity=<rgba 0 0 0 0.33>}
   slot {class=CanvasPanelSlot, properties {layoutData {alignment=<0 0>, anchors {maximum=<1 1>, minimum=<0 0>}, offsets=<ltrb 0 0 0 0>}}}
  }
  WBP_SparksCameraTimeline {
   widgetClassPath=&/WBP_SparksCameraTimeline
   bIsVariable=true
   parent=CanvasPanel_43
   widget=WBP_SparksCameraTimeline
   slot {
    class=CanvasPanelSlot
    properties {layoutData {alignment=<0 0>, anchors {maximum=<1 1>, minimum=<0 1>}, offsets=<ltrb 162.0741 -173.0811 49.2373 112.0721>}}
   }
  }
 ]
}
